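# Physics asset for the bone bow mesh, covering the bones of its decorative feathers.
# Each of five feathers is a two-segment chain of capsule bodies, with its first segment jointed to a shared master feather body.
# Every joint uses the same narrow swing limits and very tight twist limit, so the feathers can sway slightly with little axial roll.
# No other part of the bow has a body here.
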PhysicsAsset /PrimalGameplay/Items/Bows/Bone/Assets/Meshes/Bone_Bow_Physics
bodies {
 dyn_feather_01_a [{shapeType=Capsule, length=1.3039, radius=0.9705, center=<2.1803 0.0739 -0.0359>, rotation=<rot 0 -84.3538 -86.3333>}]
 dyn_feather_01_b [{shapeType=Capsule, length=2.4342, radius=0.8925, center=<2.4379 0.0994 -0.0522>, rotation=<rot 0 -84.3538 -86.3333>}]
 dyn_feather_02_a [{shapeType=Capsule, length=1.4977, radius=0.7768, center=<1.3387 -0.0093 0.0186>, rotation=<rot 0 -84.3538 -86.3333>}]
 dyn_feather_02_b [{shapeType=Capsule, length=1.4977, radius=0.7768, center=<1.3387 -0.0093 0.0186>, rotation=<rot 0 -84.3538 -86.3333>}]
 dyn_feather_03_a [{shapeType=Capsule, length=1.4977, radius=0.7768, center=<1.3387 -0.0093 0.0186>, rotation=<rot 0 -84.3538 -86.3333>}]
 dyn_feather_03_b [{shapeType=Capsule, length=1.4978, radius=0.9416, center=<1.8523 0.0415 -0.0144>, rotation=<rot 0 -84.3538 -86.3333>}]
 dyn_feather_04_a [{shapeType=Capsule, length=0.837, radius=0.6222, center=<0.9064 -0.052 0.0465>, rotation=<rot 0 -84.3538 -86.3333>}]
 dyn_feather_04_b [{shapeType=Capsule, length=0.837, radius=0.6222, center=<1.2112 -0.0219 0.0269>, rotation=<rot 0 -84.3538 -86.3333>}]
 dyn_feather_05_a [{shapeType=Capsule, length=0.7612, radius=0.6808, center=<1.1962 -0.0234 0.0278>, rotation=<rot 0 -84.3538 -86.3333>}]
 dyn_feather_05_b [{shapeType=Capsule, length=1.4977, radius=0.7768, center=<1.7497 0.0313 -0.0078>, rotation=<rot 0 -84.3538 -86.3333>}]
 dyn_feathers_MSTR [{shapeType=Capsule, length=0.505, radius=0.505, rotation=<rot 0 -90 0>}]
}
constraints [
 {bone1Name=dyn_feather_05_b, bone2Name=dyn_feather_05_a, swing1LimitDegrees=20, swing2LimitDegrees=20, twistLimitDegrees=2}
 {bone1Name=dyn_feather_02_b, bone2Name=dyn_feather_02_a, swing1LimitDegrees=20, swing2LimitDegrees=20, twistLimitDegrees=2}
 {bone1Name=dyn_feather_03_b, bone2Name=dyn_feather_03_a, swing1LimitDegrees=20, swing2LimitDegrees=20, twistLimitDegrees=2}
 {bone1Name=dyn_feather_04_b, bone2Name=dyn_feather_04_a, swing1LimitDegrees=20, swing2LimitDegrees=20, twistLimitDegrees=2}
 {bone1Name=dyn_feather_01_b, bone2Name=dyn_feather_01_a, swing1LimitDegrees=20, swing2LimitDegrees=20, twistLimitDegrees=2}
 {bone1Name=dyn_feather_01_a, bone2Name=dyn_feathers_MSTR, swing1LimitDegrees=20, swing2LimitDegrees=20, twistLimitDegrees=2}
 {bone1Name=dyn_feather_04_a, bone2Name=dyn_feathers_MSTR, swing1LimitDegrees=20, swing2LimitDegrees=20, twistLimitDegrees=2}
 {bone1Name=dyn_feather_03_a, bone2Name=dyn_feathers_MSTR, swing1LimitDegrees=20, swing2LimitDegrees=20, twistLimitDegrees=2}
 {bone1Name=dyn_feather_02_a, bone2Name=dyn_feathers_MSTR, swing1LimitDegrees=20, swing2LimitDegrees=20, twistLimitDegrees=2}
 {bone1Name=dyn_feather_05_a, bone2Name=dyn_feathers_MSTR, swing1LimitDegrees=20, swing2LimitDegrees=20, twistLimitDegrees=2}
]
